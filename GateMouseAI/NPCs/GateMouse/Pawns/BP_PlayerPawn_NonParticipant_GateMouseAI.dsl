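; Pawn Blueprint for a non-participant GateMouse AI NPC, extending a parent pawn's start and death handling.
; On BeginPlay it slows the NPC animation speed scalar and, if the animation interface is valid, the locomotion play rate multiplier to slightly below normal.
; On death it runs the parent handling, then deactivates the Boss_SpecialActor component.
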(event EventBeginPlay
  (bind _returnvalue (Animation|GetAnimationInterface))
  (|CallParentFunction)
  (CallFunction|SetNPCAnimSpeedScalar 0.9)
  (Utilities|IsValid _returnvalue
    (:Is%20Valid
      (Locomotion|SetLocomotionPlayRateMultiplier _returnvalue 0.9))))

(event EventOnDied_Event
  (params (DamagedActor :type Actor%20Object%20Reference) (Damage :type Float%20%28single-precision%29) (InstigatedBy :type Controller%20Object%20Reference) (DamageCauser :type Actor%20Object%20Reference) (HitLocation :type Vector :default "0, 0, 0") (FHitComponent :type Primitive%20Component%20Object%20Reference) (BoneName :type Name :default "None") (Momentum :type Vector :default "0, 0, 0"))
  ; On Death Deactivate Special Actor
  (|CallParentFunction DamagedActor Damage InstigatedBy DamageCauser HitLocation FHitComponent BoneName Momentum)
  (Components|Activation|Deactivate (Variables|Default|GetBoss_SpecialActor)))

(fn ConstructionScript
  (|CallParentFunction))
